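; B_Decal_SlapSplash is a decal actor for the SlapSplash consumable that shows an impact caustic effect.
; Its construction script creates a dynamic instance of MI_Decal_SlapSplash_ImpactCaustic, stores it as Decal_MID and applies it to the Decal component.
; On BeginPlay, the FadeIn timeline plays from start and writes its curve value to the material's FadeIn scalar parameter.
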(event EventBeginPlay
  (bind (_direction _newtrack_0) (Timeline|FadeIn PlayFromStart)
    (:Update
      (Rendering|Material|SetScalarParameterValue (Variables|Default|GetDecal_MID) "FadeIn" _newtrack_0 :!class /Script/Engine.MaterialInstanceDynamic))))

(fn ConstructionScript
  (bind _returnvalue (Rendering|Material|CreateDynamicMaterialInstance "/ChronoConsumables/Assets/SlapSplash/FX/MI_Decal_SlapSplash_ImpactCaustic.MI_Decal_SlapSplash_ImpactCaustic" "None" "None" :!class /Script/Engine.KismetMaterialLibrary))
  (bind _output_get (Variables|Default|SetDecal_MID _returnvalue))
  (Rendering|Components|Decal|SetDecalMaterial (Variables|Decal|GetDecal) _output_get))
